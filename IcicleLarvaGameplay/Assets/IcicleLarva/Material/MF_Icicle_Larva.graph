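# Material function for icy, crystalline Icicle Larva surfaces, outputting subsurface-style material attributes and a curvature texture.
# Triplanar crack normals and parallax-offset interior crack texture give depth, while a curvature map provides edge-highlight remapping.
# Emissive combines sky-light reflection samples with optional distance- or vertex-color glow, Fresnel masks and inverse eye adaptation.
# Optional core darkening measures from a hard-coded world position; the custom-primitive-data glow center and several experimental nodes are unconnected.
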MaterialFunction /IcicleLarvaGameplay/Assets/IcicleLarva/Material/MF_Icicle_Larva
graph {nodes [
  n1 = ScalarParameter() {defaultValue=-0.0616, group=InteriorCracks, parameterName=InteriorCracks_BumpOffset}
  n2 = CameraVectorWS()
  n3 = ComponentMask(None=Transform(None=@n2) {transformSourceType=TRANSFORMSOURCE_World, transformType=TRANSFORM_Tangent}) {g=true, r=true}
  n4 = Transform(None=MaterialFunctionCall(Normal=NamedRerouteUsage(), Flatness=ScalarParameter() {defaultValue=-1, group=InteriorCracks, parameterName=InteriorDistortion Normal}) {materialFunction=&FlattenNormal}.Result)
  n5 = Add(
   A=Multiply(A=TextureCoordinate(), B=ScalarParameter() {defaultValue=0.4796, group=InteriorCracks, parameterName=InteriorCracks_Tiling, sortPriority=1})
   B=Divide(A=Multiply(A=@n1, B=@n3), B=DotProduct(A=@n2, B=@n4))
  )
  n6 = NamedRerouteUsage()
  Add(
   A=@n5
   B=Multiply(
    A=ComponentMask(None=@n6) {g=true, r=true}
    B=ScalarParameter() {defaultValue=-0.079, group=InteriorCracks, parameterName=DistortInteriorByNormal}
   )
  )
  Add()
  n7 = VectorParameter() {group=Shading, parameterName=Color, sortPriority=1, defaultValue=<rgba 0 0.068 0.1042 1>}
  n8 = NamedRerouteUsage()
  n9 = NamedRerouteUsage()
  n10 = VectorParameter() {group=Emissive, parameterName=EmissiveGlow, defaultValue=<rgba 1.2986 2.63 5 1>}
  n11 = Power(
   Base=LinearInterpolate(
    A=ScalarParameter() {defaultValue=1, group=Emissive, parameterName=GlowVertexColor_A}
    B=ScalarParameter() {group=Emissive, parameterName=GlowVertexColor_B}
    Alpha=VertexColor().R
   )
   Exp=ScalarParameter() {defaultValue=2, group=Emissive, parameterName=GlowVertexColor_Falloff}
  )
  n12 = LocalPosition()
  n13 = Multiply(
   A=ScalarParameter() {defaultValue=0.3859, group=Emissive, parameterName=GlowAttenuation}
   B=Divide(
    A=ObjectRadius()
    B=MaterialFunctionCall() {materialFunction=&ObjectScale}.Scale Z
   )
  )
  n14 = MaterialFunctionCall(Float3=NamedRerouteUsage()) {materialFunction=&BreakOutFloat3Components}
  n15 = Multiply(A=ScalarParameter() {defaultValue=-4.8763, group=Emissive, parameterName=DistortGlow_fromCrackNormals}, B=Subtract(A=@n14.R, B=@n14.R))
  n16 = Distance(A=Constant3Vector() {desc=WorldSpace Location, constant=<rgba 1708.7944 604.3872 69.4823 0>}, B=WorldPosition().XYZ)
  n17 = StaticSwitchParameter(
   True=Divide(A=Length(None=@n12.XYZ) {is=class MaterialExpressionLength}, B=Add(A=@n13, B=@n15))
   False=Divide(A=@n16, B=ScalarParameter() {defaultValue=860.6788, group=Emissive, parameterName=GlowAttenuationWS})
  ) {defaultValue=true, group=Emissive, parameterName=GlowDistanceLocal}
  n18 = Divide(A=Constant() {r=1}, B=Multiply(A=@n17, B=@n17))
  n19 = Multiply(
   A=NamedRerouteUsage()
   B=Min(
    A=StaticSwitchParameter(True=@n11, False=@n18) {group=Emissive, parameterName=GlowUseVertexColorForGradient}
    B=ScalarParameter() {defaultValue=1, group=Emissive, parameterName=MaxGlow}
   )
  )
  n20 = SkyLightEnvMapSample(Direction=ReflectionVectorWS(), Roughness=ScalarParameter() {defaultValue=0.75, group=EnvSkySample, parameterName=SkySampleRoughness})
  n21 = ScalarParameter() {defaultValue=3, group=EnvSkySample, parameterName=SkySampleContrast}
  n22 = Multiply(
   A=NamedRerouteUsage()
   B=Multiply(A=Power(Base=@n20, Exp=@n21), B=ScalarParameter() {defaultValue=3, group=EnvSkySample, parameterName=SkySampleBrightness})
  )
  n23 = StaticSwitchParameter(True=Add(A=Multiply(A=@n10.RGB, B=@n19), B=@n22), False=@n22) {defaultValue=true, parameterName=UseGlow}
  n24 = Multiply(A=Add(A=NamedRerouteUsage(), B=Multiply(A=@n9, B=@n23)), B=ScalarParameter() {defaultValue=0.376, parameterName=BrightnessMaster})
  n25 = Multiply(A=@n24, B=NamedRerouteUsage())
  n26 = Multiply(
   A=Add(A=@n22, B=VectorParameter() {group=Shading, parameterName=SSSColor, sortPriority=3, defaultValue=<rgba 0.0225 0.0436 0.0573 1>}.RGB)
   B=ScalarParameter() {defaultValue=0.05, parameterName=SSSBrightnessMaster}
  )
  n27 = NamedRerouteUsage()
  n28 = SetMaterialAttributes!(
   MaterialAttributes
   BaseColor=StaticSwitchParameter(True=Multiply(A=@n7.RGB, B=@n8), False=@n7.RGB) {parameterName=UseCoreDarkening}
   Metallic
   Specular=ScalarParameter() {defaultValue=0.5, group=Shading, parameterName=Specular}
   Roughness=ScalarParameter() {group=Shading, parameterName=Roughness}
   EmissiveColor=EyeAdaptationInverse(LightValueInput=StaticSwitchParameter(True=@n25, False=@n24) {parameterName=UseCoreDarkening}).EyeAdaptationInverse
   Opacity=ScalarParameter() {defaultValue=0.696, group=Shading, parameterName=SSSOpacity}
   Normal=NamedRerouteUsage()
   SubsurfaceColor=StaticSwitchParameter(True=Multiply(A=@n26, B=@n27), False=@n26) {parameterName=UseCoreDarkening}
  ) {attributeSetTypes [
    #1
    #2
    #3
    #4
    #5
    #6
    #7
    #8
   ]}
  FunctionOutput(None=@n28) {outputName=SetMaterialAttributes}
  n29 = TextureSampleParameter2D() {
   group=Curvature Edge Highlights
   parameterName=CurvatureMap
   texture=&/Crystal_bake_principledshader_BaseColor
   channelNames {a="", b="", g="", r=Curvature}
  }
  FunctionOutput(None=@n29.Curvature) {outputName=CurvatureTexture, sortPriority=20}
  MaterialFunctionCall!(HeightTexture, TextureCoordinates, ReferencePlane, HeightRatioInput, CameraVector) {materialFunction=/Engine/Functions/Engine_MaterialFunctions02/Utility/BumpOffset_Advanced.BumpOffset_advanced}
  n30 = Multiply(A=ReflectionVectorWS()) {constB=-1}
  n31 = ScalarParameter() {defaultValue=217.786, group=InteriorCracks, parameterName=InteriorCracks_Depth}
  MaterialFunctionCall(
   TextureObject=TextureObject() {texture=&/Crystal_M}
   TextureSize=ScalarParameter() {defaultValue=2275.3975, group=InteriorCracks, parameterName=InteriorCracks_Size, sortPriority=2}
   WorldPosition=Add(A=WorldPosition().XYZ, B=Multiply(A=@n30, B=@n31))
  ) {materialFunction=&WorldAlignedTexture}
  n32 = OneMinus(None=@n29.Curvature)
  n33 = ScalarParameter() {defaultValue=0.4579, group=Curvature Edge Highlights, parameterName=Curvature_Glow_Contrast}
  NamedRerouteDeclaration(None=LinearInterpolate(
    A=ScalarParameter() {defaultValue=1.5602, group=Curvature Edge Highlights, parameterName=Curvature_Glow_Low}
    B=ScalarParameter() {defaultValue=0.2927, group=Curvature Edge Highlights, parameterName=Curvature_Glow_High}
    Alpha=Power(Base=@n32, Exp=@n33)
   )) {name=CurvatureRemapped, nodeColor=<rgba 1 0 0.3516 1>}
  n34 = ScalarParameter() {defaultValue=0.3747, group=Emissive, parameterName=GlowFresnel}
  n35 = PixelNormalWS()
  n36 = Constant() {r=1}
  n37 = NamedRerouteDeclaration(None=StaticSwitchParameter(True=Fresnel(ExponentIn=@n34, Normal=@n35), False=@n36) {defaultValue=true, parameterName=UseGlowFresnelMult}) {name=Fresnel, nodeColor=<rgba 0.9922 1 0 1>}
  NamedRerouteDeclaration(None=StaticSwitchParameter(True=OneMinus(None=@n37), False=@n36) {defaultValue=true, parameterName=UseGlowFresnelMult}) {name=Fresnel_Inverted, nodeColor=<rgba 0.9922 1 0 1>}
  n38 = MaterialFunctionCall(
   TextureObject=TextureObject() {samplerType=SAMPLERTYPE_Normal, texture=&/Crystal_N}
   TextureSize=ScalarParameter() {defaultValue=7833.3149, group=Cracks, parameterName=CrackNormals_Size}
   WorldSpace=StaticBool()
   ProjectionTransitionContrast=ScalarParameter() {defaultValue=50, group=Cracks, parameterName=Cracks_ProjectionContrast}
  ) {materialFunction=&WorldAlignedNormal}
  NamedRerouteDeclaration(None=MaterialFunctionCall(
    Normal=NamedRerouteDeclaration(None=@n38.XYZ Texture) {name=NormalMapTriplanar, nodeColor=<rgba 0 1 0.3672 1>}
    Flatness=ScalarParameter() {defaultValue=0.8, group=Cracks, parameterName=Cracks_Normal_Flatten}
   ) {materialFunction=&FlattenNormal}.Result) {name=NormalMap_Output, nodeColor=<rgba 0 1 0.3672 1>}
  n39 = VectorParameter() {group=Shading, parameterName=CrackColor, sortPriority=2, defaultValue=<rgba 0.2217 0.6426 1 1>}
  n40 = LinearInterpolate(
   A=ScalarParameter() {defaultValue=0.7665, group=InteriorCracks, parameterName=InteriorContrast_low}
   B=ScalarParameter() {defaultValue=4.6517, group=InteriorCracks, parameterName=InteriorContrast_high}
   Alpha=TextureSample(UVs=@n5) {texture=&Crystal_M}.R
  )
  NamedRerouteDeclaration(None=Multiply(A=Multiply(A=@n39.RGB, B=@n40), B=NamedRerouteUsage())) {name=InteriorCracks, nodeColor=<rgba 0 0.2734 1 1>}
  n41 = Power(
   Base=Divide(A=@n16, B=ScalarParameter() {defaultValue=860.6788, group=Emissive, parameterName=CoreDarken_Distance})
   Exp=ScalarParameter() {defaultValue=3, group=Emissive, parameterName=CoreDarken_Power}
  )
  NamedRerouteDeclaration(None=Max(A=Saturate(None=@n41), B=ScalarParameter() {defaultValue=0.05, group=Emissive, parameterName=CoreDarken_Min})) {name=CoreDarkenMask, nodeColor=<rgba 1 0 0.0469 1>}
  PixelNormalWS()
  n42 = ScalarParameter() {defaultValue=5722.0576, group=Emissive, parameterName=Test}
  Saturate(None=Divide(A=Subtract(A=@n16, B=@n42)) {constB=1})
  n43 = VectorParameter() {bUseCustomPrimitiveData=true, parameterName=GlowCenter_StoredInLocal, primitiveDataIndex=1, defaultValue=<rgba 0 0 0 1>}
  TransformPosition(None=@n43.RGB) {desc=store in local Position In BP, transformSourceType=TRANSFORMPOSSOURCE_Instance, transformType=TRANSFORMPOSSOURCE_World}
  VertexNormalWS()
  VertexNormalWS()
 ]}
