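# Subsurface material for the Icicle Larva, usable on Nanite meshes and Niagara mesh particles, with optional snow cover.
# MF_Icicle_Larva supplies the base attributes, which the UseSnow switch blends with the snow layer from the Hermes landscape-match function.
# HeightLerp drives the blend from a world-normal top-down mask and an inverted world-aligned snow texture.
# Sine-noise, distance-field and radial gradient mask branches are present but do not feed that blend.
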Material /IcicleLarvaGameplay/Assets/IcicleLarva/Material/M_Icicle_Larva
graph {
 properties {
  bAutomaticallySetUsageInEditor=false
  bUseMaterialAttributes=true
  bUsedWithNanite=true
  bUsedWithNiagaraMeshParticles=true
  shadingModel=MSM_Subsurface
  usedShadingModels=Subsurface
 }
 nodes [
  n1 = MaterialFunctionCall(Float3=Divide(A=WorldPosition().XYZ, B=ScalarParameter() {defaultValue=1185.8717, parameterName=RandomSnowNoise_Size})) {materialFunction=&BreakOutFloat3Components}
  n2 = Multiply(A=Multiply(A=Sine(None=@n1.R), B=Sine(None=@n1.G)), B=Sine(None=@n1.B))
  n3 = Multiply(A=Divide(A=Add(A=@n2)), B=ScalarParameter() {defaultValue=0.8335, parameterName=RandomSnowNoise_Mult})
  n4 = ScalarParameter() {defaultValue=0.8017, parameterName=RandomSnowNoise_Max}
  n5 = Max(A=@n3, B=@n4)
  n6 = WorldPosition()
  n7 = VertexNormalWS()
  n8 = ScalarParameter() {defaultValue=35.905, parameterName=BaseDistance}
  n9 = ScalarParameter() {defaultValue=496.6618, parameterName=Radius}
  n10 = Saturate(None=Divide(A=LocalPosition().Z, B=ScalarParameter() {defaultValue=1255.3562, parameterName=BottomFade}))
  n11 = FeatureLevelSwitch(Default=Min(A=DistanceFieldApproxAO!(World Position=@n6.XYZ, Normal=@n7, BaseDistance=@n8, Radius=@n9), B=@n10), ES3_1=@n10)
  Add(A=@n5, B=@n11)
  n12 = ComponentMask(None=VertexNormalWS()) {b=true}
  Divide(A=@n12)
  Min(A=@n3, B=@n4)
  n13 = Add(A=@n6.XYZ, B=Multiply(A=VertexNormalWS(), B=@n9))
  Saturate(None=Divide(A=DistanceToNearestSurface!(World Position=@n13), B=@n8))
  n14 = Saturate(None=LinearInterpolate(A=ScalarParameter() {parameterName=TopDownSnow_low}, B=ScalarParameter() {defaultValue=0.256, parameterName=TopDownSnow_high}, Alpha=@n12))
  Saturate(None=Multiply(A=Multiply(A=@n5, B=@n11), B=OneMinus(None=@n14)))
  n15 = MaterialFunctionCall!(Grass_UseCustomSlopeMask, Grass_CustomSlopeMask) {materialFunction=/BRCoreMaterials/Hermes/Terrain/Landscape/Materials/Layers/MF_Hermes_StaticMeshLandscapeLayerMatch}
  n16 = MaterialFunctionCall() {materialFunction=&/MF_Icicle_Larva}
  n17 = VertexNormalWS()
  n18 = Saturate(None=LinearInterpolate(
    A=ScalarParameter() {defaultValue=1, parameterName=Snow_TopDown_low}
    B=ScalarParameter() {defaultValue=0.5285, parameterName=Snow_TopDown_high}
    Alpha=ComponentMask(None=@n17) {b=true}
   ))
  n19 = TextureObject() {texture=&/Snow_M}
  n20 = ScalarParameter() {defaultValue=772.0963, parameterName=SnowSIze}
  n21 = OneMinus(None=ComponentMask(None=MaterialFunctionCall(TextureObject=@n19, TextureSize=@n20) {materialFunction=&WorldAlignedTexture}.XYZ Texture) {r=true})
  n22 = ScalarParameter() {parameterName=Snow TransitionContrast}
  StaticSwitchParameter(
   True=BlendMaterialAttributes(
    A=NamedRerouteDeclaration(None=@n15.SnowLayer) {name=Landscape_Snow, nodeColor=<rgba 0 1 0.1094 1>}
    B=@n16.SetMaterialAttributes
    Alpha=MaterialFunctionCall(Transition Phase=@n18, Height Texture=@n21, Contrast=@n22) {materialFunction=&HeightLerp}.Alpha
   )
   False=@n16.SetMaterialAttributes
  ) {defaultValue=true, group=Snow, parameterName=UseSnow}
  n23 = StaticSwitchParameter(True=Constant(), False=@n18) {parameterName=UseSnowTopDown}
  n24 = Divide(
   A=Length(None=LocalPosition().XYZ) {is=class MaterialExpressionLength}
   B=Multiply(A=ObjectRadius(), B=ScalarParameter() {defaultValue=1255.3562, parameterName=BottomFade})
  )
  StaticSwitchParameter(True=Add(A=@n23, B=Saturate(None=@n24)), False=@n23) {defaultValue=true, parameterName=UseSnowGradientBlend}
 ]
}
